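; B_StorageWars_DJBooth is a Blueprint for a DJ booth building actor in the Storage Wars diegetic music set of BattleWoods.
; On BeginPlay it turns off damage and marks the building actor invulnerable, so the booth cannot be destroyed.
; The construction script is empty, and the file shows no audio logic.
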(event EventBeginPlay
  (Variables|Actor|SetbCanBeDamaged false)
  (Variables|BuildingActor|SetbIsInvulnerable true))

(fn ConstructionScript)
